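; Gameplay ability for the Tidal Crane Carapace medallion that passively regenerates the owning Fort pawn's shield.
; On activation it starts a looping timer that adds shield up to a maximum, cleared when the ability ends.
; A damage-received gameplay event blocks regeneration for a configurable delay, which repeat hits do not restart.
; A heal gameplay cue is added while regenerating and removed when blocked or full.
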(event Ability|EventActivateAbility
  (bind _outfortplayerpawn (CallFunction|GetAvatarasValidFortPlayerPawn)
    (:ValidFortPlayerPawn
      (Variables|Default|SetFort%20Player%20Pawn _outfortplayerpawn)
      (Ability|Tasks|WaitGameplayEvent "(TagName=\"Event.Damage.Received\")"
        (:then
          (bind _returnvalue (Utilities|Time|SetTimerbyEvent (delegate Custom|RegenShield) (ScalableFloat|GetValueatLevel (Variables|HFValues|GetTimeBetweenHeals)) true))
          (Variables|Default|SetHealingTimer _returnvalue))
        (:EventReceived
          (CallFunction|BlockRegen))))))

(event Custom|Block%20Regen
  (Variables|Default|SetBlockRegen true)
  (CallFunction|RemoveCue)
  (Utilities|FlowControl|Delay (ScalableFloat|GetValueatLevel (Variables|HFValues|GetRegenDelayAfterDamage)))
  (Variables|Default|SetBlockRegen false))

(event Custom|RegenShield
  (if (Variables|Default|GetBlockRegen)
    _
    (elif (>= (FortPawn|GetShield (Variables|Default|GetFort%20Player%20Pawn)) (Variables|HFValues|GetMaxShield))
      (CallFunction|RemoveCue)
      (else
        (CallFunction|AddCue)
        (FortPawn|SetShield (Variables|Default|GetFort%20Player%20Pawn) (+ (FortPawn|GetShield (Variables|Default|GetFort%20Player%20Pawn)) (Variables|HFValues|GetShieldHealAmount)))))))

(event Ability|EventOnEndAbility
  (params (bWasCancelled :type Boolean))
  (Utilities|Time|ClearandInvalidateTimerbyHandle (Variables|Default|GetHealingTimer)))

(fn AddCue
  (bind _healgc (Variables|Default|GetHealGC))
  (bind _returnvalue (Ability|GameplayCue|IsGameplayCueActive (Variables|Default|GetFort%20Player%20Pawn) _healgc))
  (if (not _returnvalue)
    (Ability|AddGameplayCueWithParamsToOwner _healgc (Ability|GameplayCue|MakeGameplayCueParameters 1.0 (ScalableFloat|GetValueatLevel (Variables|HFValues|GetTimeBetweenHeals))))))

(fn RemoveCue
  (Ability|RemoveGameplayCueFromOwner (Variables|Default|GetHealGC)))
